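# Translucent, unlit, two-sided material for SafariGnome shoe effects, set up for Niagara mesh particles and ribbons.
# Particle color feeds emissive adjustments, while opacity combines panning noise, sine UV masks, particle alpha and view-facing falloff, depth-faded off mobile.
# World position offset math adds scrolling sine waves to clamped drag from particle position, both weighted along U.
# Dynamic parameters drive UV offset, wave strength and blend; the drag target repeats X on all axes.
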Material /BRCosmetics/Characters/Player/Male/Medium/Bodies/M_MED_SafariGnome/FX/M_Character_SafariGnome_ShoesFX
graph {
 properties {
  bAutomaticallySetUsageInEditor=false
  bUsedWithNiagaraMeshParticles=true
  bUsedWithNiagaraRibbons=true
  blendMode=BLEND_Translucent
  shadingModel=MSM_Unlit
  twoSided=true
  usedShadingModels=Unlit
 }
 nodes [
  Constant()
  n1 = DynamicParameter() {defaultValue=<rgba 0 1 1 1>, paramNames [OpacityScroll, Offset, WPOMult, WPOLerp]}
  n2 = Constant()
  n3 = ComponentMask(None=Add(A=TextureCoordinate() {uTiling=2.2}, B=AppendVector(A=@n1.Offset, B=@n2))) {r=true}
  n4 = Time()
  n5 = Add(A=@n3, B=Multiply(A=@n4) {constB=-0.7})
  n6 = Add(A=@n3, B=Multiply(A=@n4) {constB=-1})
  n7 = Multiply(
   A=Multiply(A=Sine(None=@n5) {period=2.2}, B=Sine(None=@n6) {period=3})
   B=VectorParameter() {parameterName=WPO Direction, defaultValue=<rgba 1 1 1 1>}.RGB
  )
  n8 = DynamicParameter() {defaultValue=<rgba 0 1 1 1>, paramNames [OpacityScroll, Offset, WPOMult, WPOLerp]}
  n9 = Multiply(A=Transform(None=@n7) {transformSourceType=TRANSFORMSOURCE_Local}, B=@n8.WPOMult)
  n10 = ComponentMask(None=TextureCoordinate()) {r=true}
  n11 = Saturate(None=Power(Base=@n10) {constExponent=1.2})
  n12 = DynamicParameter() {parameterIndex=1, defaultValue=<rgba 1 1 1 0>, paramNames [X,Y,Z,WPOIntensity]}
  n13 = ScalarParameter() {defaultValue=13, group=WPO, parameterName=WPO Max Drag, sliderMax=50}
  n14 = Clamp(
   None=Subtract(
    A=MaterialFunctionCall(X=@n12.X, Y=@n12.X, Z=@n12.X) {materialFunction=&MakeFloat3}.Result
    B=ParticlePositionWS()
   )
   Min=Multiply(A=@n13) {constB=-1}
   Max=@n13
  )
  n15 = Power(Base=ComponentMask(None=TextureCoordinate()) {r=true}, Exp=ScalarParameter() {defaultValue=0.8, parameterName=WPOBlendPower})
  n16 = Multiply(A=@n12.WPOIntensity, B=@n15)
  LinearInterpolate(B=Add(A=Multiply(A=@n9, B=@n11), B=LinearInterpolate(B=@n14, Alpha=@n16)), Alpha=@n8.WPOLerp)
  n17 = ParticleColor()
  MaterialFunctionCall(Emissive=@n17.RGB) {materialFunction=&MF_FX_EmissiveAdjustments}
  Multiply(A=@n10)
  n18 = Abs(None=DotProduct(A=CameraVectorWS(), B=VertexNormalWS()))
  Power(Base=@n18) {constExponent=1.4}
  Saturate(None=@n15)
  ScalarParameter() {parameterName=Change Transform}
  n19 = TextureCoordinate()
  n20 = Sine(None=ComponentMask(None=@n19) {g=true}) {period=2}
  n21 = Multiply(A=@n20, B=Sine(None=ComponentMask(None=@n19) {r=true}) {period=2})
  n22 = Add(A=TextureCoordinate(), B=@n1.Offset)
  n23 = Constant2Vector() {g=1.6, r=0.7}
  n24 = TextureSample(UVs=Panner(Coordinate=Multiply(A=@n22, B=@n23)) {speedX=-0.8}) {samplerType=SAMPLERTYPE_LinearGrayscale, texture=/Engine/Functions/Engine_MaterialFunctions02/ExampleContent/Textures/LowResBlurredNoise}
  n25 = Power(Base=@n20) {constExponent=15}
  n26 = Add(A=@n24.R, B=Multiply(A=Saturate(None=@n25)) {constB=0.15})
  n27 = Power(Base=@n25) {constExponent=3}
  n28 = Add(A=Saturate(None=Power(Base=@n26) {constExponent=3}), B=Multiply(A=Saturate(None=@n27)) {constB=0.3})
  n29 = Multiply(A=Multiply(A=Multiply(A=@n21), B=Multiply(A=@n28, B=@n17.A)), B=Saturate(None=@n18))
  ShadingPathSwitch(Default=DepthFade(Opacity=@n29, FadeDistance=Constant() {r=3}) {fadeDistanceDefault=10}, Mobile=@n29)
  Transform(None=@n7) {transformSourceType=TRANSFORMSOURCE_World, transformType=TRANSFORM_Camera}
 ]
}
