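# Translucent lit glass base material for the SafariGnome player body, with a night-reactive look.
# MF_Glass attributes blend toward TOD base color, emissive and Fresnel opacity by enabled night value or ForceNighttime; the simulated timestamp cycle does not drive this blend.
# Masked thin film follows, then optional camera dither.
# A switch selects vertex hiding, via world position offset toward the object pivot, or character opacity, before TechArt overrides.
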Material /BRCosmetics/Characters/Player/Male/Medium/Bodies/M_MED_SafariGnome/Materials/M_SafariGnome_Glass_Base
graph {
 properties {
  bEnableResponsiveAA=true
  bScreenSpaceReflections=true
  bUseMaterialAttributes=true
  bUsedWithMeshDeformer=true
  bUsedWithMorphTargets=true
  bUsedWithSkeletalMesh=true
  blendMode=BLEND_Translucent
  ditherOpacityMask=true
  pixelDepthOffsetMode=PDOM_Legacy
  translucencyDirectionalLightingIntensity=5
  translucencyLightingMode=TLM_SurfacePerPixelLighting
  usedShadingModels=Default Lit
 }
 nodes [
  n1 = Constant()
  n2 = MaterialFunctionCall(Value_Morning=@n1, Value_Day=@n1, Value_Evening=@n1, Value_Night=Constant() {r=1}) {materialFunction=&Scalar_Time_of_Day}
  LinearInterpolate(A=@n2.Result, B=NamedRerouteUsage(), Alpha=Saturate(None=ScalarParameter() {group=TOD Reactivity, parameterName=TOD_Timestamp}))
  n3 = TextureSampleParameter2D() {
   group=00_Glass
   parameterName=TechArtMask
   samplerType=SAMPLERTYPE_Masks
   sortPriority=-96
   texture=&T_BlackMask
  }
  n4 = MaterialFunctionCall!(TechArtMask=@n3.RGB) {materialFunction=/Game/Global/Materials/Character_Parent_2023/Glass_Parent_2023/MF_Glass}
  n5 = Multiply(A=@n2.Result, B=ScalarParameter() {group=TOD Reactivity, parameterName=EnableReactivity, sortPriority=0})
  n6 = ScalarParameter() {group=TOD Reactivity, parameterName=ForceNighttime, sortPriority=1}
  n7 = TextureSampleParameter2D() {
   group=TOD Reactivity
   parameterName=TOD BaseColor
   samplerType=SAMPLERTYPE_Masks
   sortPriority=0
   texture=&T_BlackMask
  }
  n8 = NamedRerouteDeclaration(None=@n7.RGB) {name=Texture_BaseColor, nodeColor=<rgba 0 1 0.4375 1>}
  n9 = NamedRerouteUsage()
  n10 = NamedRerouteUsage()
  n11 = MaterialFunctionCall!(
   Material Attributes=BlendMaterialAttributes(
    A=@n4.Result
    B=SetMaterialAttributes!(MaterialAttributes=@n4.Result, Base Color=@n8, Emissive Color=@n9, Opacity=@n10) {attributeSetTypes [#1, #2, #3]}
    Alpha=Max(A=@n5, B=@n6)
   )
   Fresnel=Fresnel() {baseReflectFraction=0, exponent=1}
   Alpha=@n3.RGB
  ) {materialFunction=/Game/Global/MaterialFunctions/TechArt/MF_ThinFilm_TechArt}
  n12 = StaticSwitchParameter(
   True=MaterialFunctionCall!(In=@n11.Result) {materialFunction=/Game/Packages/Fortress_SharedMaterials/Base_Material_Functions/BaseCustomizationFunctions/MF_FN_Vehicle_CameraDither}.Result
   False=@n11.Result
  ) {group=00_Glass, parameterName=UseCameraDither, sortPriority=250}
  n13 = Multiply(A=Subtract(A=ObjectPositionWS(), B=WorldPosition().XYZ), B=ScalarParameter() {group=00_Glass, parameterName=VertexHide, sortPriority=301})
  MaterialFunctionCall(In=StaticSwitchParameter(
    True=SetMaterialAttributes!(MaterialAttributes=@n12, World Position Offset=@n13) {attributeSetTypes [#4]}
    False=MaterialFunctionCall!(Material=@n12) {materialFunction=/Game/Global/Materials/Character_Parent_2023/Functions/BaseFN_CharacterOpacity_2023}.Result
   ) {group=00_Glass, parameterName=UseVertexHide, sortPriority=300}) {materialFunction=&MF_TechArtOverrides}
  n14 = TextureSampleParameter2D() {
   group=TOD Reactivity
   parameterName=TOD Emissive
   samplerType=SAMPLERTYPE_Masks
   sortPriority=0
   texture=&T_BlackMask
  }
  n15 = VectorParameter() {group=TOD Reactivity, parameterName=TOD EmissiveTint, defaultValue=<rgba 0 0 0 1>}
  NamedRerouteDeclaration(None=Multiply(A=Add(A=@n14.RGB, B=@n15.RGB), B=ScalarParameter() {group=TOD Reactivity, parameterName=TOD EmissiveStrength, sortPriority=15})) {name=OUT_Emissive, nodeColor=<rgba 0.3594 1 0 1>}
  n16 = Multiply(
   A=Fresnel(ExponentIn=ScalarParameter() {defaultValue=1, group=TOD Reactivity, parameterName=TOD FresnelExponentTransparency, sortPriority=30})
   B=ScalarParameter() {defaultValue=1, group=TOD Reactivity, parameterName=TOD GlassThickness, sortPriority=31}
  )
  NamedRerouteDeclaration(None=LinearInterpolate(
    A=ScalarParameter() {group=TOD Reactivity, parameterName=TOD InnerTransparency}
    B=ScalarParameter() {defaultValue=0.85, group=TOD Reactivity, parameterName=TOD OuterTransparency, sortPriority=33}
    Alpha=Saturate(None=@n16)
   )) {name=OUT_Opacity, nodeColor=<rgba 1 0 0.6797 1>}
  n17 = ScalarParameter() {group=TOD Reactivity, parameterName=TOD_Timestamp}
  n18 = Subtract(A=Time(), B=@n17)
  n19 = Saturate(None=@n17)
  n20 = Multiply(A=Add(A=Multiply(A=@n18, B=@n19)) {constB=-12}, B=Divide() {constB=30, desc=One loop every 30 seconds})
  n21 = ConstantBiasScale(None=Multiply(A=Sine(None=@n20)) {constB=2})
  NamedRerouteDeclaration(None=SmoothStep(Value=Saturate(None=@n21))) {name=Simulated_TOD, nodeColor=<rgba 0.5 1 0 1>}
 ]
}
