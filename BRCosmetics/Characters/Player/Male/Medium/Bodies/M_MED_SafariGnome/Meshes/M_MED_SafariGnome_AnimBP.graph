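# Post-process style AnimBP for the SafariGnome player body cosmetic, layering shorts and shoulder-pad motion onto the parent mesh's copied pose.
# A Control Rig takes skydive, airborne, emote and DBNO inputs, then additive bone tweaks and spring-driven AnimDynamics chains follow, split by a cached pose.
# EventGraph derives state floats from player flags and weapon type.
# Its weapon, crouch/sprint, gravity and sim-space outputs are not wired into the AnimGraph.
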AnimBlueprint /BRCosmetics/Characters/Player/Male/Medium/Bodies/M_MED_SafariGnome/Meshes/M_MED_SafariGnome_AnimBP
target_skeleton=&/M_MED_SafariGnome_Skeleton
graphs [
 AnimGraph {nodes [
   n1 = Anim Dynamics - Bone: thigh_r_short_dyn!(
    ComponentPose=Local To Component(LocalPose=Use cached pose 'Initial Cache'())
    Alpha=1
    GravityOverride:Vector="0,0,-980"
    ExternalForce="0,0,0"
   ) -> (Pose) {
    angularDampingOverride=1
    angularSpringConstant=7500
    bAngularSpring=true
    bOverrideAngularDamping=true
    bUseGravityOverride=true
    gravityScale=0.35
    simulationSpace=RootRelative
    alphaBoolBlend {blendOption=Linear}
    boundBone {boneName=thigh_r_short_dyn}
    componentAppliedLinearAccClamp=<3000 3000 3000>
    componentLinearAccScale=<0.5 0.5 0.5>
    componentLinearVelScale=<0.5 0.5 0.5>
    gravityOverride {z=-980}
    physicsBodyDefinitions [{
      boundBone {boneName=thigh_r_short_dyn}
      boxExtents=<20 20 20>
      constraintSetup {angularLimitsMax=<5 8 -2>, angularLimitsMin=<-5 -8 -3>, angularTarget=<1 0 0>}
      localJointOffset=<4 0 0>
     }]
   }
   n2 = Anim Dynamics - Bone: shorts_r_outter!(ComponentPose=@n1, Alpha=1, GravityOverride:Vector="0,0,-980", ExternalForce="0,0,0") -> (Pose) {
    angularDampingOverride=1
    angularSpringConstant=5000
    bAngularSpring=true
    bOverrideAngularDamping=true
    bUseGravityOverride=true
    gravityScale=0.35
    simulationSpace=RootRelative
    alphaBoolBlend {blendOption=Linear}
    boundBone {boneName=shorts_r_outter}
    componentAppliedLinearAccClamp=<3000 3000 3000>
    componentLinearAccScale=<0.35 0.35 0.35>
    componentLinearVelScale=<0.35 0.35 0.35>
    gravityOverride {z=-980}
    physicsBodyDefinitions [{
      boundBone {boneName=shorts_r_outter}
      boxExtents=<30 40 45>
      constraintSetup {angularLimitsMax=<5 5 0>, angularLimitsMin=<-5 -20 0>, angularTarget=<1 0 0>}
      localJointOffset=<4 0 0>
     }]
   }
   n3 = Anim Dynamics - Bone: shorts_r_back!(ComponentPose=@n2, Alpha=1, GravityOverride:Vector="0,0,-980", ExternalForce="0,0,0") -> (Pose) {
    angularDampingOverride=1
    angularSpringConstant=5000
    bAngularSpring=true
    bOverrideAngularDamping=true
    bUseGravityOverride=true
    gravityScale=0.35
    simulationSpace=RootRelative
    alphaBoolBlend {blendOption=Linear}
    boundBone {boneName=shorts_r_back}
    componentAppliedLinearAccClamp=<3000 3000 3000>
    componentLinearAccScale=<0.35 0.35 0.35>
    componentLinearVelScale=<0.35 0.35 0.35>
    gravityOverride {z=-980}
    physicsBodyDefinitions [{
      boundBone {boneName=shorts_r_back}
      boxExtents=<30 40 45>
      constraintSetup {angularLimitsMax=<5 0 25>, angularLimitsMin=<-5 0 -4>, angularTarget=<1 0 0>}
      localJointOffset=<4 0 0>
     }]
   }
   n4 = Anim Dynamics - Bone: shorts_r_inner!(ComponentPose=@n3, Alpha=1, GravityOverride:Vector="0,0,-980", ExternalForce="0,0,0") -> (Pose) {
    angularDampingOverride=1
    angularSpringConstant=5000
    bAngularSpring=true
    bOverrideAngularDamping=true
    bUseGravityOverride=true
    gravityScale=0.35
    simulationSpace=RootRelative
    alphaBoolBlend {blendOption=Linear}
    boundBone {boneName=shorts_r_inner}
    componentAppliedLinearAccClamp=<3000 3000 3000>
    componentLinearAccScale=<0.35 0.35 0.35>
    componentLinearVelScale=<0.35 0.35 0.35>
    gravityOverride {z=-980}
    physicsBodyDefinitions [{
      boundBone {boneName=shorts_r_inner}
      boxExtents=<30 40 45>
      constraintSetup {angularLimitsMax=<5 4 0>, angularLimitsMin=<-5 -8 0>, angularTarget=<1 0 0>}
      localJointOffset=<4 0 0>
     }]
   }
   n5 = Anim Dynamics - Bone: shorts_r_front!(ComponentPose=@n4, Alpha=1, GravityOverride:Vector="0,0,-980", ExternalForce="0,0,0") -> (Pose) {
    angularDampingOverride=1
    angularSpringConstant=5000
    bAngularSpring=true
    bOverrideAngularDamping=true
    bUseGravityOverride=true
    gravityScale=0.35
    simulationSpace=RootRelative
    alphaBoolBlend {blendOption=Linear}
    boundBone {boneName=shorts_r_front}
    componentAppliedLinearAccClamp=<3000 3000 3000>
    componentLinearAccScale=<0.35 0.35 0.35>
    componentLinearVelScale=<0.35 0.35 0.35>
    gravityOverride {z=-980}
    physicsBodyDefinitions [{
      boundBone {boneName=shorts_r_front}
      boxExtents=<30 40 45>
      constraintSetup {angularLimitsMax=<5 0 -5>, angularLimitsMin=<-5 0 -20>, angularTarget=<1 0 0>}
      localJointOffset=<4 0 0>
     }]
   }
   n6 = "Transform (Modify) Bone - Bone: dyn_shoulderpad_02_l"(ComponentPose=@n5, Alpha=1, Translation="0,-0.75,0", Rotation="0,0,0", Scale="1,1,1") {
    rotationMode=BMM_Additive
    rotationSpace=BCS_BoneSpace
    scaleMode=BMM_Additive
    translationMode=BMM_Additive
    translationSpace=BCS_BoneSpace
    alphaBoolBlend {blendOption=Linear}
    boneToModify {boneName=dyn_shoulderpad_02_l}
    translation {y=-0.75}
   }
   n7 = "Transform (Modify) Bone - Bone: dyn_shoulderpad_02_r"(ComponentPose=@n6, Alpha=1, Translation="0,0.75,0", Rotation="0,0,0", Scale="1,1,1") {
    rotationMode=BMM_Additive
    rotationSpace=BCS_BoneSpace
    scaleMode=BMM_Additive
    translationMode=BMM_Additive
    translationSpace=BCS_BoneSpace
    alphaBoolBlend {blendOption=Linear}
    boneToModify {boneName=dyn_shoulderpad_02_r}
    translation {y=0.75}
   }
   n8 = "Transform (Modify) Bone - Bone: dyn_shoulderpad_02_l"(ComponentPose=@n7, Alpha=1, Translation="0,-1,0", Rotation="0,0,0", Scale="1,1,1") {
    rotationMode=BMM_Additive
    rotationSpace=BCS_BoneSpace
    scaleMode=BMM_Additive
    translationMode=BMM_Additive
    translationSpace=BCS_BoneSpace
    alphaBoolBlend {blendOption=Linear}
    boneToModify {boneName=dyn_shoulderpad_02_l}
    translation {y=-1}
   }
   n9 = "Transform (Modify) Bone - Bone: dyn_shoulderpad_02_r"(ComponentPose=@n8, Alpha=1, Translation="0,0.75,0", Rotation="0,0,0", Scale="1,1,1") {
    rotationMode=BMM_Additive
    rotationSpace=BCS_BoneSpace
    scaleMode=BMM_Additive
    translationMode=BMM_Additive
    translationSpace=BCS_BoneSpace
    alphaBoolBlend {blendOption=Linear}
    boneToModify {boneName=dyn_shoulderpad_02_r}
    translation {y=0.75}
   }
   Output Pose(Result=Component To Local(ComponentPose=@n9)) {layerGroup=None}
   n10 = Control Rig!(
    Source=Copy Pose From Mesh() {bUseAttachedParent=true}
    Alpha
    is_Skydive:double=Get skydiveFloat() -> (skydiveFloat:double)
    is_on_Air:double=Get Is_on_Air() -> (Is_on_Air:double)
    is_playing_Emote:double=Get Is playing Emote() -> (Is playing Emote:double)
    Is DBNO:bool=Get bIsDBNO()
   ) -> (Pose) {
    alphaBoolBlend {blendOption=Linear}
    controlRigAssetReference {blueprintRigClass=&/M_MED_SafariGnome_CtrlRig.M_MED_SafariGnome_CtrlRig_C}
   }
   n11 = "Transform (Modify) Bone - Bone: base_shoulderpad_l"(ComponentPose=Local To Component(LocalPose=@n10), Alpha=1, Translation="2,0,0.5", Rotation="13,0,0", Scale="0.95,0.95,0.95") {
    rotationMode=BMM_Additive
    scaleMode=BMM_Additive
    translationMode=BMM_Additive
    alphaBoolBlend {blendOption=Linear}
    boneToModify {boneName=base_shoulderpad_l}
    rotation {pitch=13}
    scale=<0.95 0.95 0.95>
    translation {x=2, z=0.5}
   }
   n12 = "Transform (Modify) Bone - Bone: base_shoulderpad_r"(ComponentPose=@n11, Alpha=1, Translation="-1.5,0,1.25", Rotation="-15,0,0", Scale="0.95,0.95,0.95") {
    rotationMode=BMM_Additive
    scaleMode=BMM_Additive
    translationMode=BMM_Additive
    alphaBoolBlend {blendOption=Linear}
    boneToModify {boneName=base_shoulderpad_r}
    rotation {pitch=-15}
    scale=<0.95 0.95 0.95>
    translation {x=-1.5, z=1.25}
   }
   n13 = "Transform (Modify) Bone - Bone: dyn_shoulderpad_l"!(ComponentPose=@n12, bAlphaBoolEnabled:bool=True, Translation="0,0,0", Rotation="0,-5,0", Scale="1,1.05,1") -> (Pose) {
    alphaInputType=Bool
    rotationMode=BMM_Additive
    rotationSpace=BCS_BoneSpace
    scaleMode=BMM_Additive
    scaleSpace=BCS_BoneSpace
    translationMode=BMM_Additive
    translationSpace=BCS_BoneSpace
    alphaBoolBlend {blendOption=Linear}
    boneToModify {boneName=dyn_shoulderpad_l}
    rotation {yaw=-5}
    scale {y=1.05}
   }
   n14 = "Transform (Modify) Bone - Bone: dyn_shoulderpad_r"!(ComponentPose=@n13, bAlphaBoolEnabled:bool=True, Translation="0,0,0.8", Rotation="0,10,-2.5", Scale="1,1.1,1") -> (Pose) {
    alphaInputType=Bool
    rotationMode=BMM_Additive
    rotationSpace=BCS_BoneSpace
    scaleMode=BMM_Additive
    scaleSpace=BCS_BoneSpace
    translationMode=BMM_Additive
    translationSpace=BCS_BoneSpace
    alphaBoolBlend {blendOption=Linear}
    boneToModify {boneName=dyn_shoulderpad_r}
    rotation {roll=-2.5, yaw=10}
    scale {y=1.1}
    translation {z=0.8}
   }
   n15 = Anim Dynamics - Bone: thigh_l_short_dyn!(ComponentPose=@n14, Alpha=1, GravityOverride:Vector="0,0,-980", ExternalForce="0,0,0") -> (Pose) {
    angularDampingOverride=1
    angularSpringConstant=7500
    bAngularSpring=true
    bOverrideAngularDamping=true
    bUseGravityOverride=true
    gravityScale=0.35
    simulationSpace=RootRelative
    alphaBoolBlend {blendOption=Linear}
    boundBone {boneName=thigh_l_short_dyn}
    componentAppliedLinearAccClamp=<3000 3000 3000>
    componentLinearAccScale=<0.5 0.5 0.5>
    componentLinearVelScale=<0.5 0.5 0.5>
    gravityOverride {z=-980}
    physicsBodyDefinitions [{
      boundBone {boneName=thigh_l_short_dyn}
      boxExtents=<20 20 20>
      constraintSetup {angularLimitsMax=<5 8 0>, angularLimitsMin=<-5 -8 0>, angularTarget=<1 0 0>}
      localJointOffset=<4 0 0>
     }]
   }
   n16 = Anim Dynamics - Bone: shorts_l_outter!(ComponentPose=@n15, Alpha=1, GravityOverride:Vector="0,0,-980", ExternalForce="0,0,0") -> (Pose) {
    angularDampingOverride=1
    angularSpringConstant=5000
    bAngularSpring=true
    bOverrideAngularDamping=true
    bUseGravityOverride=true
    gravityScale=0.35
    simulationSpace=RootRelative
    alphaBoolBlend {blendOption=Linear}
    boundBone {boneName=shorts_l_outter}
    componentAppliedLinearAccClamp=<3000 3000 3000>
    componentLinearAccScale=<0.35 0.35 0.35>
    componentLinearVelScale=<0.35 0.35 0.35>
    gravityOverride {z=-980}
    physicsBodyDefinitions [{
      boundBone {boneName=shorts_l_outter}
      boxExtents=<30 40 45>
      constraintSetup {angularLimitsMax=<5 5 0>, angularLimitsMin=<-5 -20 0>, angularTarget=<1 0 0>}
      localJointOffset=<4 0 0>
     }]
   }
   n17 = Anim Dynamics - Bone: shorts_l_back!(ComponentPose=@n16, Alpha=1, GravityOverride:Vector="0,0,-980", ExternalForce="0,0,0") -> (Pose) {
    angularDampingOverride=1
    angularSpringConstant=5000
    bAngularSpring=true
    bOverrideAngularDamping=true
    bUseGravityOverride=true
    gravityScale=0.35
    simulationSpace=RootRelative
    alphaBoolBlend {blendOption=Linear}
    boundBone {boneName=shorts_l_back}
    componentAppliedLinearAccClamp=<3000 3000 3000>
    componentLinearAccScale=<0.35 0.35 0.35>
    componentLinearVelScale=<0.35 0.35 0.35>
    gravityOverride {z=-980}
    physicsBodyDefinitions [{
      boundBone {boneName=shorts_l_back}
      boxExtents=<30 40 45>
      constraintSetup {angularLimitsMax=<5 0 8>, angularLimitsMin=<-5 0 -20>, angularTarget=<1 0 0>}
      localJointOffset=<4 0 0>
     }]
   }
   n18 = Anim Dynamics - Bone: shorts_l_inner!(ComponentPose=@n17, Alpha=1, GravityOverride:Vector="0,0,-980", ExternalForce="0,0,0") -> (Pose) {
    angularDampingOverride=1
    angularSpringConstant=5000
    bAngularSpring=true
    bOverrideAngularDamping=true
    bUseGravityOverride=true
    gravityScale=0.35
    simulationSpace=RootRelative
    alphaBoolBlend {blendOption=Linear}
    boundBone {boneName=shorts_l_inner}
    componentAppliedLinearAccClamp=<3000 3000 3000>
    componentLinearAccScale=<0.35 0.35 0.35>
    componentLinearVelScale=<0.35 0.35 0.35>
    gravityOverride {z=-980}
    physicsBodyDefinitions [{
      boundBone {boneName=shorts_l_inner}
      boxExtents=<30 40 45>
      constraintSetup {angularLimitsMax=<5 5 0>, angularLimitsMin=<-5 -8 0>, angularTarget=<1 0 0>}
      localJointOffset=<4 0 0>
     }]
   }
   n19 = Anim Dynamics - Bone: shorts_l_front!(ComponentPose=@n18, Alpha=1, GravityOverride:Vector="0,0,-980", ExternalForce="0,0,0") -> (Pose) {
    angularDampingOverride=1
    angularSpringConstant=5000
    bAngularSpring=true
    bOverrideAngularDamping=true
    bUseGravityOverride=true
    gravityScale=0.35
    simulationSpace=RootRelative
    alphaBoolBlend {blendOption=Linear}
    boundBone {boneName=shorts_l_front}
    componentAppliedLinearAccClamp=<3000 3000 3000>
    componentLinearAccScale=<0.35 0.35 0.35>
    componentLinearVelScale=<0.35 0.35 0.35>
    gravityOverride {z=-980}
    physicsBodyDefinitions [{
      boundBone {boneName=shorts_l_front}
      boxExtents=<30 40 45>
      constraintSetup {angularLimitsMax=<5 0 5>, angularLimitsMin=<-5 0 -5>, angularTarget=<1 0 0>}
      localJointOffset=<4 0 0>
     }]
   }
   Save cached pose 'Initial Cache'(Pose=Component To Local(ComponentPose=@n19)) {cacheName=Initial Cache}
  ]}
 EventGraph {nodes [
   Event BlueprintUpdateAnimation()
   > n1 = Sequence!(execute) -> (then_0, then_1, then_2:Exec, then_3:Exec, then_4:Exec, then_5:Exec, then_6:Exec)
   > n2 = Cast To FortPlayerAnimInstance(Object=GetAnimInstance(self=GetOwnerSkeletalMeshComponent()))
   > Set Short_Rifle_Out(Short_Rifle_Out:bool=false) -> (Output_Get:bool)
   > Set PistolOut(PistolOut:bool=false) -> (Output_Get:bool)
   > Set rifle_out(rifle_out:bool=false) -> (Output_Get:bool)
   > n3 = Switch on EFortWeaponCoreAnimation!(execute, Selection:EFortWeaponCoreAnimation Enum=Get WeaponCoreAnim(self=@n2)) -> (
    Melee:Exec
    Pistol:Exec
    Shotgun:Exec
    PaperBlueprint:Exec
    Rifle:Exec
    MeleeOneHand:Exec
    MachinePistol:Exec
    RocketLauncher:Exec
    GrenadeLauncher:Exec
    GoingCommando:Exec
    AssaultRifle:Exec
    TacticalShotgun:Exec
    SniperRifle:Exec
    TrapPlacement:Exec
    ShoulderLauncher:Exec
    AbilityDecoTool:Exec
    Crossbow:Exec
    C4:Exec
    RemoteControl:Exec
    DualWield:Exec
    AR_BullPup:Exec
    AR_ForwardGrip:Exec
    MedPackPaddles:Exec
    SMG_P90:Exec
    AR_DrumGun:Exec
    Consumable_Small:Exec
    Consumable_Large:Exec
    Balloon:Exec
    MountedTurret:Exec
    CreativeTool:Exec
    ExplosiveBow:Exec
    AshtonIndigo:Exec
    AshtonChicago:Exec
    MeleeDualWield:Exec
    Unarmed:Exec
   )
   Set PistolOut(execute=@n3.Pistol, PistolOut=true)
   > Set Short_Rifle_Out(Short_Rifle_Out=true)
   > Set rifle_out(rifle_out=true)
   > Set large_gun_out(large_gun_out:double="To Float (Boolean)"(InBool=Get rifle_out() -> (rifle_out:bool))) -> (Output_Get:double)
   > Set PistolGunOut(PistolGunOut:double="To Float (Boolean)"(InBool=Get PistolOut() -> (PistolOut:bool))) -> (Output_Get:double)
   > Set ShortRifle Gun Out(ShortRifle Gun Out:double="To Float (Boolean)"(InBool=Get Short_Rifle_Out() -> (Short_Rifle_Out:bool))) -> (Output_Get:double)
   n4 = Branch(execute=@n1.then_3, Condition=OR Boolean(A=Get bIsSkydiveFloating(), B=Get bIsSkyDiving()))
   > Set Is_on_Air(Is_on_Air:double=1) -> (Output_Get:double)
   Set Is_on_Air(execute=@n4.else, Is_on_Air=0)
   n5 = Get bIsTargeting()
   n6 = Get Pistol01() -> (Pistol01:bool)
   Set PistolGunOut(execute=@n1.then_1, PistolGunOut="To Float (Boolean)"(InBool=AND Boolean(A=@n6, B=@n5)))
   n7 = Gravity Override State Function DEPRECATED(
    execute=@n1.then_2
    CharacterPartAnimInstance=Self-Reference()
    GameTimeInSeconds=GetGameTimeInSeconds()
    GravityOverrideParams=Get Gravity Override Params()
    RBAN Params=Get RBAN Params()
   )
   > Set GravityOverride(GravityOverride=@n7)
   > Set Sim Space Settings(Sim Space Settings:Sim Space Settings Structure=Break RBANSetting Struct!(RBANSettingStruct:"RBANSetting Struct Structure (by ref)"=@n7.RBAN SettingsOut) -> (
     ComponentLinearAccScale_28_#1:Vector=~
     ComponentLinearVelScale_31_#2:Vector=~
     SimSpaceSettings_34_#3:Sim Space Settings Structure
    ) {is=title K2Node_BreakStruct}.SimSpaceSettings_34_#3) -> (Output_Get:Sim Space Settings Structure)
   Set skydiveFloat(execute=@n1.then_4, skydiveFloat:double="To Float (Boolean)"(InBool=Get bIsSkyDiving())) -> (Output_Get:double)
   Set Is playing Emote(execute=@n1.then_5, Is playing Emote:double="To Float (Boolean)"(InBool=Get bIsPlayingEmote())) -> (Output_Get:double)
   n8 = Get bIsCrouchSprinting()
   n9 = Get bIsCrouching()
   n10 = Get bIsSprinting()
   n11 = Get bIsMoving2D()
   Set Crounching and sprinting(
    execute=@n1.then_6
    Crounching and sprinting:double="To Float (Boolean)"(InBool=OR Boolean!(A=@n8, B=@n9, C:bool=@n10, D:bool=@n11) -> (ReturnValue))
   ) -> (Output_Get:double)
   n12 = GetAnimInstance(self=GetOwnerSkeletalMeshComponent())
   Switch on EFortWeaponCoreAnimation(Selection=Get WeaponCoreAnim(self=Cast To FortPlayerAnimInstance(Object=@n12)))
   Set Pistol01(Pistol01:bool=true) -> (Output_Get:bool)
  ]}
]
variables [
 Crounching and sprinting {type=number}
 Gravity Override Params {type=GravityOverrideParamsStruct}
 GravityOverride {type=Vector}
 Is Celebration Emote {type=number}
 Is playing Emote {type=number}
 Is_on_Air {type=number}
 Iscelebration {type=boolean}
 IsCelebrationEmote {type=number}
 isPistol01_Out {type=number}
 large_gun_out {type=number}
 Pistol01 {type=boolean}
 PistolGunOut {type=number}
 PistolOut {type=boolean}
 RBAN Params {
  type=RBANParamsStruct
  default {
   dBNO {simSpaceSettings {
     maxAngularAcceleration=1000
     maxAngularVelocity=3
     maxLinearAcceleration=300
     maxLinearVelocity=700
     velocityScaleZ=0
     externalLinearDragV=<0 0.5 0>
    }}
   dBNO_Carried {simSpaceSettings {
     maxAngularAcceleration=1000
     maxAngularVelocity=3
     maxLinearAcceleration=300
     maxLinearVelocity=700
     velocityScaleZ=0
     externalLinearDragV=<0 0.5 0>
    }}
   emoteOrMelee {simSpaceSettings {
     maxAngularAcceleration=1000
     maxAngularVelocity=3
     maxLinearAcceleration=300
     maxLinearVelocity=700
     velocityScaleZ=0
     externalLinearDragV=<0 0.5 0>
    }}
   front_End {simSpaceSettings {
     maxAngularAcceleration=1000
     maxAngularVelocity=3
     maxLinearAcceleration=300
     maxLinearVelocity=700
     velocityScaleZ=0
     externalLinearDragV=<0 0.5 0>
    }}
   onGround_Crouch_Moving {simSpaceSettings {
     maxAngularAcceleration=1000
     maxAngularVelocity=3
     maxLinearAcceleration=300
     maxLinearVelocity=700
     velocityScaleZ=0
     externalLinearDragV=<0 0.5 0>
    }}
   onGround_Falling {simSpaceSettings {
     maxAngularAcceleration=1000
     maxAngularVelocity=3
     maxLinearAcceleration=300
     maxLinearVelocity=700
     velocityScaleZ=0
     externalLinearDragV=<0 0.5 0>
    }}
   onGround_Moving {simSpaceSettings {
     maxAngularAcceleration=1000
     maxAngularVelocity=3
     maxLinearAcceleration=300
     maxLinearVelocity=700
     velocityScaleZ=0
     externalLinearDragV=<0 0.5 0>
    }}
   onGround_Standing {simSpaceSettings {
     maxAngularAcceleration=1000
     maxAngularVelocity=3
     maxLinearAcceleration=300
     maxLinearVelocity=700
     velocityScaleZ=0
     externalLinearDragV=<0 0.5 0>
    }}
   rail_Grinding {simSpaceSettings {
     maxAngularAcceleration=1000
     maxAngularVelocity=3
     maxLinearAcceleration=300
     maxLinearVelocity=700
     velocityScaleZ=0
     worldAlpha=0.1
     externalLinearDragV=<0 0.5 0>
    }}
   riding_Moving {simSpaceSettings {
     maxAngularAcceleration=1000
     maxAngularVelocity=3
     maxLinearAcceleration=300
     maxLinearVelocity=700
     velocityScaleZ=0
     externalLinearDragV=<0 0.5 0>
    }}
   riding_Moving_Targeting {simSpaceSettings {
     maxAngularAcceleration=1000
     maxAngularVelocity=3
     maxLinearAcceleration=300
     maxLinearVelocity=700
     velocityScaleZ=0
     externalLinearDragV=<0 0.5 0>
    }}
   riding_Standing {simSpaceSettings {
     maxAngularAcceleration=1000
     maxAngularVelocity=3
     maxLinearAcceleration=300
     maxLinearVelocity=700
     velocityScaleZ=0
     externalLinearDragV=<0 0.5 0>
    }}
   skydive_DivingDown {simSpaceSettings {
     maxAngularAcceleration=1000
     maxAngularVelocity=3
     maxLinearAcceleration=300
     maxLinearVelocity=700
     velocityScaleZ=0
     externalLinearDragV=<0 0.5 0>
    }}
   skydive_FloatingDown {
    componentLinearAccScale=<0 0.5 0>
    componentLinearVelScale=<0 0.5 0>
    simSpaceSettings {
     maxAngularAcceleration=1000
     maxAngularVelocity=3
     maxLinearAcceleration=300
     maxLinearVelocity=700
     velocityScaleZ=0
     externalLinearDragV=<0 0.5 0>
    }
   }
   skydive_GoingUp {simSpaceSettings {
     maxAngularAcceleration=1000
     maxAngularVelocity=3
     maxLinearAcceleration=300
     maxLinearVelocity=700
     velocityScaleZ=0
     externalLinearDragV=<0 0.5 0>
    }}
   skydive_Parachuting {
    componentLinearAccScale=<0 0.1 0.3>
    componentLinearVelScale=<0 0.1 0.3>
    simSpaceSettings {
     maxAngularAcceleration=1000
     maxAngularVelocity=3
     maxLinearAcceleration=300
     maxLinearVelocity=700
     velocityScaleZ=0
     externalLinearDragV=<0 0.5 0>
    }
   }
   skydive_Surfing {simSpaceSettings {
     maxAngularAcceleration=1000
     maxAngularVelocity=3
     maxLinearAcceleration=300
     maxLinearVelocity=700
     velocityScaleZ=0
     externalLinearDragV=<0 0.5 0>
    }}
   swimming {simSpaceSettings {
     maxAngularAcceleration=1000
     maxAngularVelocity=3
     maxLinearAcceleration=300
     maxLinearVelocity=700
     velocityScaleZ=0
     externalLinearDragV=<0 0.5 0>
    }}
   tactical_Sprint {simSpaceSettings {
     maxAngularAcceleration=1000
     maxAngularVelocity=3
     maxLinearAcceleration=300
     maxLinearVelocity=700
     velocityScaleZ=0
     externalLinearDragV=<0 0.5 0>
    }}
   ziplining {simSpaceSettings {
     maxAngularAcceleration=1000
     maxAngularVelocity=3
     maxLinearAcceleration=300
     maxLinearVelocity=700
     velocityScaleZ=0
     externalLinearDragV=<0 0.5 0>
    }}
  }
 }
 rifle_out {type=boolean}
 Short_Rifle_Out {type=boolean}
 ShortRifle Gun Out {type=number}
 Sim Space Settings {
  type=SimSpaceSettings
  default {
   dampingAlpha=1
   maxAngularAcceleration=10000
   maxAngularVelocity=10000
   maxLinearAcceleration=10000
   maxLinearVelocity=10000
   velocityScaleZ=1
   worldAlpha=0
   externalLinearDragV=<0 0 0>
  }
 }
 skydiveFloat {type=number}
]
