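# Material function for the SafariGnome character body that adds optional anisotropic shading to incoming material attributes.
# One texture supplies the tangent in RGB, remapped from a -1 lower bound, and an anisotropy mask in alpha, scaled by a max weight.
# An optional static switch uses a Fortnite quality switch to give all tiers below Epic a separate weight.
# The UseAnisotropicShading static switch either sets Tangent and Anisotropy or passes the attributes through unchanged.
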MaterialFunction /BRCosmetics/Characters/Player/Male/Medium/Bodies/M_MED_SafariGnome/Materials/MF_Anisotropic_TechArt_SafariGnome
graph {nodes [
  n1 = TextureSampleParameter2D() {
   group=TOD Anisotropic
   parameterName=TOD AnisotropicTangentWeight
   samplerType=SAMPLERTYPE_LinearColor
   texture=/Game/Global/Textures/T_Default_Aniso
  }
  n2 = LinearInterpolate(A=Constant3Vector() {constant=<rgba -1 -1 -1 0>}, Alpha=@n1.RGB)
  n3 = ScalarParameter() {defaultValue=1, group=TOD Anisotropic, parameterName=TOD AnisotropyMaxWeight_StandardQuality}
  n4 = ScalarParameter() {defaultValue=1, group=TOD Anisotropic, parameterName=TOD AnisotropyMaxWeight, sliderMax=1}
  n5 = Multiply(
   A=@n1.A
   B=StaticSwitchParameter(
    True=MaterialFunctionCall!(
     Low _ PC Low and Med=@n3
     "High _ XB1, PS4"=@n3
     "Epic _ PC High, XSX, PS5"=@n4
     --------
     "Mobile Low _ PC Perf, Android"=@n3
     Mobile Med _ Switch=@n3
    ) {materialFunction=/Game/Global/MaterialFunctions/TechArt/MF_Fortnite_QualitySwitch_Scalar}.Result
    False=@n4
   ) {group=TOD Anisotropic, parameterName=TOD UseSeperateAnisoWeightForLowEnd}
  )
  n6 = FunctionInput(Preview=SetMaterialAttributes!(MaterialAttributes)) {bUsePreviewValueAsDefault=true, inputType=FunctionInput_MaterialAttributes}
  FunctionOutput(None=StaticSwitchParameter(
    True=SetMaterialAttributes!(MaterialAttributes=@n6, Tangent=@n2, Anisotropy=@n5) {attributeSetTypes [#1, #2]}
    False=@n6
   ) {group=TOD Anisotropic, parameterName=TOD UseAnisotropicShading, sortPriority=0})
 ]}
