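# Material function for the SafariGnome character body that optionally layers a clear coat onto incoming material attributes.
# It builds a clear-coat variant with full coat strength, multiplied metallic and the ClearCoat shading model.
# One parameter crossfades which layer, coat or base, takes the incoming roughness versus the UndercoatRoughness value.
# A selectable channel of the Alpha input masks the blend, and a static switch can pass the material through unchanged.
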MaterialFunction /BRCosmetics/Characters/Player/Male/Medium/Bodies/M_MED_SafariGnome/Materials/MF_ClearCoat_TechArt_SafariGnome
graph {nodes [
  n1 = GetMaterialAttributes(None=FunctionInput() {inputName=Material, inputType=FunctionInput_MaterialAttributes, sortPriority=1}) {attributeGetTypes [#1, #2]}
  n2 = ScalarParameter() {defaultValue=0.1, group=TOD ClearCoat, parameterName=TOD UndercoatRoughness}
  n3 = ScalarParameter() {group=TOD ClearCoat, parameterName=TOD RoughnessMapAffectClearCoatRoughness}
  n4 = SetMaterialAttributes!(
   MaterialAttributes=@n1.MaterialAttributes
   CustomData0=Constant() {r=1}
   CustomData1=LinearInterpolate(A=@n2, B=@n1.Roughness, Alpha=@n3)
   Roughness=LinearInterpolate(A=@n1.Roughness, B=@n2, Alpha=@n3)
   Metallic=Multiply(A=@n1.Metallic, B=ScalarParameter() {defaultValue=1, group=TOD ClearCoat, parameterName=TOD UndercoatMetallicMultiplier})
   ShadingModel=ShadingModel() {shadingModel=MSM_ClearCoat}
  ) {attributeSetTypes [
    #3
    #4
    #1
    #2
    #5
   ]}
  n5 = ChannelMaskParameter(None=FunctionInput() {bUsePreviewValueAsDefault=true, inputName=Alpha, sortPriority=2}) {group=TOD ClearCoat, parameterName=TOD ClearcoatMaskChannel}
  FunctionOutput(None=StaticSwitchParameter(True=BlendMaterialAttributes(A=@n1.MaterialAttributes, B=@n4, Alpha=@n5.R), False=@n1.MaterialAttributes) {group=TOD ClearCoat, parameterName=TOD UseClearCoat, sortPriority=0})
 ]}
