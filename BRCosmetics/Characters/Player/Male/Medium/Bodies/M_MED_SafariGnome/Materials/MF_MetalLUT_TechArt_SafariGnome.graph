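# Material function for the SafariGnome character that adds a view-angle-dependent metal reflectance tint to base color.
# The clamped dot product of camera vector and pixel normal samples an aluminium reflectance curve from the character curve atlas.
# One selected channel multiplies BaseColor, scaled by an intensity parameter.
# An Alpha input's selected channel masks the blend into the incoming material attributes, with a static switch to toggle the effect.
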MaterialFunction /BRCosmetics/Characters/Player/Male/Medium/Bodies/M_MED_SafariGnome/Materials/MF_MetalLUT_TechArt_SafariGnome
graph {
 properties {libraryCategoriesText []}
 nodes [
  n1 = NamedRerouteUsage()
  n2 = GetMaterialAttributes(None=FunctionInput() {inputName=Materials, inputType=FunctionInput_MaterialAttributes}) {attributeGetTypes [#1]}
  n3 = CameraVectorWS()
  n4 = PixelNormalWS()
  n5 = CurveAtlasRowParameter(CurveTime=Clamp(None=DotProduct(A=@n3, B=@n4))) {
   atlas=/Game/Global/Textures/Default/Curve/FN_Character_Atlas
   curve=/Game/Characters/Player/Female/Medium/Bodies/F_MED_Ms_Alpine/Textures/AluminiumreflectanceCurve
   group=TOD MetalLUT
   parameterName=TOD_Metal LUT Curve
  }
  n6 = ChannelMaskParameter(None=@n5) {group=TOD MetalLUT, parameterName=TOD_LUTChannel}
  n7 = ScalarParameter() {defaultValue=1, group=TOD MetalLUT, parameterName=TOD_MetalLutIntensity}
  n8 = SetMaterialAttributes!(
   MaterialAttributes=@n2.MaterialAttributes
   BaseColor=LinearInterpolate(A=@n2.BaseColor, B=Multiply(A=@n2.BaseColor, B=@n6.R), Alpha=Saturate(None=@n7))
  ) {attributeSetTypes [#1]}
  n9 = ChannelMaskParameter(None=FunctionInput() {bUsePreviewValueAsDefault=true, inputName=Alpha, sortPriority=1}) {group=TOD MetalLUT, parameterName=TOD_MetalLUTMaskChannel}
  FunctionOutput(None=StaticSwitchParameter(True=BlendMaterialAttributes(A=@n1, B=@n8, Alpha=@n9.R), False=NamedRerouteUsage()) {group=TOD MetalLUT, parameterName=TOD_UseMetalLUT})
  NamedRerouteDeclaration(None=@n2.MaterialAttributes) {name=MaterialAttributes, nodeColor=<rgba 0.9219 0 1 1>}
 ]
}
