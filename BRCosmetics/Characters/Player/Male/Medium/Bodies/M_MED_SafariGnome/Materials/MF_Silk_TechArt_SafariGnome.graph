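# Material function for the SafariGnome character that layers a parameterized silk look onto incoming material attributes.
# A Fresnel term blends brightened base color toward an edge tint, shifts metallic between two values and scales roughness at edges, faking silk sheen.
# It also sets the shading model and CustomData0.
# A channel-selectable Alpha mask blends the result over the original, and the TOD_UseSilk static switch bypasses the effect when off.
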MaterialFunction /BRCosmetics/Characters/Player/Male/Medium/Bodies/M_MED_SafariGnome/Materials/MF_Silk_TechArt_SafariGnome
graph {nodes [
  n1 = NamedRerouteUsage()
  n2 = FunctionInput() {inputName=Material, inputType=FunctionInput_MaterialAttributes}
  n3 = GetMaterialAttributes(None=@n2) {attributeGetTypes [
    #1
    #2
    #3
    #4
   ]}
  n4 = Multiply(A=@n3.BaseColor, B=ScalarParameter() {defaultValue=1, group=TOD Silk, parameterName=TOD_SilkBaseColorBrightness})
  n5 = VectorParameter() {group=TOD Silk, parameterName=TOD_SilkEdgeTint, defaultValue=<rgba 0.1094 0.1094 0.1094 1>}
  n6 = Fresnel(ExponentIn=ScalarParameter() {defaultValue=2, group=TOD Silk, parameterName=TOD_Silk Fresnel}) {baseReflectFraction=0}
  n7 = ScalarParameter() {defaultValue=1, group=TOD Silk, parameterName=TOD_SilkFresnelMax}
  n8 = ScalarParameter() {group=TOD Silk, parameterName=TOD_SilkFresnelMin}
  n9 = ScalarParameter() {defaultValue=0.5, group=TOD Silk, parameterName=TOD_SilkEdgeAniso}
  n10 = SetMaterialAttributes!(
   MaterialAttributes=@n3.MaterialAttributes
   BaseColor=Saturate(None=LinearInterpolate(A=@n4, B=@n5.RGB, Alpha=@n6))
   Metallic=Saturate(None=LinearInterpolate(A=@n7, B=@n8, Alpha=@n6))
   Specular=@n3.MaterialAttributes
   Roughness=LinearInterpolate(A=@n3.Roughness, B=Multiply(A=@n3.Roughness, B=@n9), Alpha=@n6)
   ShadingModel=ShadingModel()
   SubsurfaceColor=@n3.MaterialAttributes
   CustomData0=Constant() {r=1}
  ) {attributeSetTypes [
    #1
    #2
    #3
    #4
    #5
    #6
    #7
   ]}
  n11 = ChannelMaskParameter(None=FunctionInput() {bUsePreviewValueAsDefault=true, inputName=Alpha, sortPriority=3}) {group=TOD Silk, parameterName=TOD_SilkMaskChannel, sortPriority=1}
  FunctionOutput(None=StaticSwitchParameter(True=BlendMaterialAttributes(A=@n1, B=@n10, Alpha=@n11.R) {vertexAttributeBlendType=UseA}, False=@n1) {group=TOD Silk, parameterName=TOD_UseSilk, sortPriority=0})
  NamedRerouteDeclaration(None=@n2) {name=Material, nodeColor=<rgba 1 0.5625 0 1>}
 ]}
